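# Niagara effect for the Dragon Cart bus that emits thin smoke and flickering embers from fixed offsets.
# Looping GPU emitter Main launches particles from four points, which TrailEmbers_EPIC and TrailEmbers_HIGH read to spawn colliding ember trails.
# The trail emitters share a setup but differ in spawn rates, apparently as quality tiers.
# Smoke_thin drifts on wind, tinted by a time-of-day NPC parameter, and smoke and embers scale with distance.
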NiagaraSystem /DragonCartNarrative/Assets/FX/NS_DragonCart_BusFX
dependencies {
 usedDataInterfaces [
  &NiagaraDataInterfaceArrayFloat3
  &NiagaraDataInterfaceCurve
  &NiagaraDataInterfaceColorCurve
  &NiagaraDataInterfaceSpriteRendererInfo
  &NiagaraDataInterfaceParticleRead
 ]
 usedDynamicInputs [&SimulationPosition, &UniformRangedFloat]
 usedModules [&SystemState]
}
emitters {
 Main {
  simTarget=GPUComputeSim
  emitterUpdateScript [
   &EmitterState(
    Life Cycle Mode=::NewEnumerator1
    Loop Duration=&RandomRangeFloat(Minimum=1, Maximum=2)
    Recalculate Duration Each Loop=true
    UseLoopDelay=true
   )
   &SpawnRate(
    SpawnRate=&FloatFromCurve(FloatCurve=DataInterface() {propertyValues {Curve {keys [<at 0 1>,<at 0.25 0.1>]}}}, CurveIndex=$Emitter.NormalizedLoopAge, Scale Curve=2)
    Use Spawn Probability=false
   )
  ]
  particleSpawnScript [
   &InitializeParticle(Color Mode=::NewEnumerator2, Sprite Size Mode=::NewEnumerator3, Uniform Sprite Size=250, UsePositionOffset=false, Write Material Random=true)
   &SampleParticlesFromOtherEmitter(
    Age Sampling:ENiagara_DirectReadSamplingMode=::NewEnumerator0
    Applied Alpha Scale=0.01
    Color Sampling=::NewEnumerator0
    Reject Particle If Sampled Index Is Invalid=false
   ) {enabled=false}
   &PartitionParticles(Number Of Partitions=4)
   &AddVelocity(
    Cone Angle=25
    Cone Axis=<0 2 0>
    Inner Cone Angle=20
    Module.UseConeFalloff=false
    Module.UseVelDistribution=false
    Override Seed=false
    Velocity Mode=::NewEnumerator2
    Velocity Speed=&RandomRangeFloat(Minimum=1, Maximum=1)
   )
   &ShapeLocation(MaskSpawnGroup=false, Offset Mode=::NewEnumerator0, Override Seed=false, Sphere Radius=1, UseOwnerScale=true)
   &SystemLocation(
    MaskSpawnGroup=false
    Offset=&SelectVectorFromArray(
     Array Sampling Mode=::NewEnumerator0
     Direct Array Index=/CascadeToNiagaraConverter/NiagaraScripts/DynamicInputScripts/SequentialInt(Use Array Data:NiagaraBool=false, First Number:NiagaraInt32=0, Sequence:NiagaraDataInterfaceArrayInt32=?, Last Number:NiagaraInt32=3)
     Vector Selection Array=DataInterface() {propertyValues {FloatData [<200 250 250>, <-100 100 250>, <200 450 50>, <-100 100 0>]}}
     Random Seed=0
    )
   )
  ]
  particleUpdateScript [
   &GravityForce(Gravity=<0 0 -250>)
   &Drag(Drag=0, Use Linear Drag=true, Use Rotational Drag=false)
   &CurlNoiseForce(Noise Frequency=5, Noise Strength=100, Pan Noise=false)
   &SolveForcesAndVelocity(Clamp Velocity=false, Limit Acceleration=false)
   &ParticleState()
   &DynamicMaterialParameters(
    Index 0 Param 1=0.8
    Index 0 Param 2=1
    Index 0 Param 3=1
    Index 1 Param 1=150
    Index 1 Param 4=1
    Module.Param0WriteEnabled=true
    Module.Param0WriteEnabled1=true
    Module.Param1WriteEnabled=true
    Module.Param1WriteEnabled1=true
    Module.Param2WriteEnabled=true
    Module.Param2WriteEnabled1=true
    Module.Param3WriteEnabled=false
    Module.Param3WriteEnabled1=true
    Write Parameter Index 1=true
   )
   &ScaleColor(
    Scale Alpha=&FloatFromCurve(
     FloatCurve=DataInterface() {propertyValues {Curve {keys [
         {arriveTangent=9.0725, interpMode=RCIM_Cubic, leaveTangent=9.0725, tangentMode=RCTM_User, time=0, value=0}
         {arriveTangent=-7.2695, interpMode=RCIM_Cubic, leaveTangent=-7.2695, tangentMode=RCTM_User, time=0.2, value=1.75}
         <at 0.8 1.8>
         {arriveTangent=-0.0405, interpMode=RCIM_Cubic, leaveTangent=-0.0405, tangentMode=RCTM_User, time=1, value=0}
        ]}}}
     Scale Curve=0.02
    )
    ScaleA=true
    ScaleRGB=true
   ) {module=ScaleColor001}
   &ScaleSpriteSize(Uniform Curve Sprite Scale=DataInterface() {propertyValues {Curve {keys [{interpMode=RCIM_Cubic, tangentMode=RCTM_User, time=0, value=1}, {interpMode=RCIM_Cubic, tangentMode=RCTM_User, time=1, value=0}]}}}) {module=ScaleSpriteSize001}
  ]
  rendererClasses [&NiagaraSpriteRendererProperties]
  renderers [{rendererClass=&NiagaraSpriteRendererProperties}]
 }
 Smoke_thin {
  emitterUpdateScript [
   &EmitterState()
   &SpawnRate(SpawnRate=10, Use Spawn Probability=false)
  ]
  particleSpawnScript [
   &InitializeParticle(
    Color=<rgba 0.0388 0.0405 0.0417 0.5>
    Lifetime Max=6
    Lifetime Min=5
    Lifetime Mode=::NewEnumerator1
    Sprite Rotation Mode=::NewEnumerator2
    Sprite Size Mode=::NewEnumerator1
    Uniform Sprite Size Max=250
    Uniform Sprite Size Min=150
    UsePositionOffset=true
    Write Material Random=true
   ) {module=InitializeParticle001}
   &ApplyInitialForces(Apply Rotational Force to Rotational Velocity=false)
   &AddVelocity(Velocity=<1 0 1>, Velocity Speed Scale=&RandomRangeFloat(Minimum=30, Maximum=50))
   &SystemLocation(
    MaskSpawnGroup=false
    Offset=&SelectVectorFromArray(
     Direct Array Index=&SequentialInt(Use Array Data:NiagaraBool=false, First Number:NiagaraInt32=0, Sequence:NiagaraDataInterfaceArrayInt32=?, Last Number:NiagaraInt32=3)
     Vector Selection Array=DataInterface() {propertyValues {FloatData [<0 0 0>, <150 0 -75>, <200 400 0>, <0 100 200>]}}
    )
   )
  ]
  particleUpdateScript [
   &ParticleState()
   &WindForce(Scale=0.3, Turbulence Flow Loop Duration=2, Wind Speed=<100 0 100>)
   &Drag(Drag=0.1, Use Linear Drag=true, Use Rotational Drag=true)
   &SolveForcesAndVelocity(Clamp Velocity=false, Limit Acceleration=false)
   &ScaleSpriteSize(Scale Factor=&Vector2DFromFloat(Value=&FloatFromCurve(
      FloatCurve=DataInterface() {propertyValues {Curve {keys [
          {arriveTangent=1.6147, interpMode=RCIM_Cubic, leaveTangent=1.6147, tangentMode=RCTM_User, time=0, value=0.2}
          {arriveTangent=0.2677, interpMode=RCIM_Cubic, leaveTangent=0.2677, tangentMode=RCTM_User, time=1, value=1}
         ]}}}
      Scale Curve=&FN_LOD_Distance_Remap(Start Distance=300, End Distance=500, Minimum Output=0.15, Maximum Output=1, Clamp Output=true)
     )))
   &DynamicMaterialParameters(
    Index 0 Param 1=&FloatFromCurve(FloatCurve=DataInterface() {propertyValues {Curve {keys [<at -0.0046 1.0113 RCIM_Cubic>,<at 1.0011 0.002>]}}})
    Index 0 Param 4=1
    Module.Param0WriteEnabled=true
    Module.Param1WriteEnabled=false
    Module.Param2WriteEnabled=false
    Module.Param3WriteEnabled=true
   ) {module=DynamicMatParams}
   &ScaleColor(
    Scale Alpha=&Multiply_Float(
     A=&FloatFromCurve(FloatCurve=DataInterface() {propertyValues {Curve {keys [
          {arriveTangent=0.4422, interpMode=RCIM_Cubic, leaveTangent=0.4422, tangentMode=RCTM_User, time=0, value=0}
          {arriveTangent=1.0142, interpMode=RCIM_Cubic, leaveTangent=1.0142, time=0.25, value=1}
          {arriveTangent=-1.4783, interpMode=RCIM_Cubic, leaveTangent=-1.4783, tangentMode=RCTM_User, time=0.6624, value=0.6717}
          <at 1.0011 0.002>
         ]}}})
     B=&FN_LOD_Distance_Remap(Start Distance=500, End Distance=1000, Minimum Output=0.15, Maximum Output=1, Clamp Output=true)
    )
    Scale RGB=&Multiply_Vector(
     A=&MakeVectorFromLinearColorRGB(Color=&ColorFromCurve(
       ColorCurve=DataInterface() {propertyValues {
         AlphaCurve {keys [<at 0 1>,<at 1 1>]}
         BlueCurve {keys [
           {
            arriveTangent=0.256
            arriveTangentWeight=0.0864
            interpMode=RCIM_Cubic
            leaveTangent=0.256
            leaveTangentWeight=0.0864
            tangentMode=RCTM_User
            tangentWeightMode=RCTWM_WeightedBoth
            time=0.001
            value=0.217
           }
           <at 0.082 0.1927>
           {
            arriveTangent=1.2546
            arriveTangentWeight=0.0556
            interpMode=RCIM_Cubic
            leaveTangent=1.2546
            leaveTangentWeight=0.0556
            tangentMode=RCTM_User
            tangentWeightMode=RCTWM_WeightedBoth
            time=0.171
            value=0.2035
           }
           <at 0.216 0.0839>
           {interpMode=RCIM_Cubic, tangentMode=RCTM_User, time=0.29, value=0.2969}
           {
            arriveTangent=1.194
            arriveTangentWeight=0.0789
            interpMode=RCIM_Cubic
            leaveTangent=1.194
            leaveTangentWeight=0.0789
            tangentMode=RCTM_User
            tangentWeightMode=RCTWM_WeightedBoth
            time=0.375
            value=0.3229
           }
           {
            arriveTangent=0.246
            arriveTangentWeight=0.0709
            interpMode=RCIM_Cubic
            leaveTangent=0.246
            leaveTangentWeight=0.0709
            tangentMode=RCTM_User
            tangentWeightMode=RCTWM_WeightedBoth
            time=0.6589
            value=0.7
           }
           {interpMode=RCIM_Cubic, tangentMode=RCTM_User, time=0.817, value=0.118}
           {arriveTangent=0.541, interpMode=RCIM_Cubic, leaveTangent=0.541, time=0.9114, value=0.1303}
           {
            arriveTangent=0.0005
            arriveTangentWeight=0.1844
            interpMode=RCIM_Cubic
            leaveTangent=0.0005
            leaveTangentWeight=0.1844
            tangentMode=RCTM_User
            tangentWeightMode=RCTWM_WeightedBoth
            time=1
            value=0.217
           }
          ]}
         GreenCurve {keys [
           {
            arriveTangent=0.2383
            arriveTangentWeight=0.0668
            interpMode=RCIM_Cubic
            leaveTangent=0.2383
            leaveTangentWeight=0.0668
            tangentMode=RCTM_User
            tangentWeightMode=RCTWM_WeightedBoth
            time=0.001
            value=0.1763
           }
           <at 0.082 0.1325>
           {
            arriveTangent=2.4228
            arriveTangentWeight=0.0324
            interpMode=RCIM_Cubic
            leaveTangent=2.4228
            leaveTangentWeight=0.0324
            tangentMode=RCTM_User
            tangentWeightMode=RCTWM_WeightedBoth
            time=0.171
            value=0.207
           }
           <at 0.216 0.0839>
           {interpMode=RCIM_Cubic, tangentMode=RCTM_User, time=0.29, value=0.2969}
           {
            arriveTangent=1.194
            arriveTangentWeight=0.0789
            interpMode=RCIM_Cubic
            leaveTangent=1.194
            leaveTangentWeight=0.0789
            tangentMode=RCTM_User
            tangentWeightMode=RCTWM_WeightedBoth
            time=0.375
            value=0.2876
           }
           {
            arriveTangent=0.0845
            arriveTangentWeight=0.0575
            interpMode=RCIM_Cubic
            leaveTangent=0.0845
            leaveTangentWeight=0.0575
            tangentMode=RCTM_User
            tangentWeightMode=RCTWM_WeightedBoth
            time=0.6589
            value=0.5947
           }
           {interpMode=RCIM_Cubic, tangentMode=RCTM_User, time=0.817, value=0.2197}
           {arriveTangent=-0.2369, interpMode=RCIM_Cubic, leaveTangent=-0.2369, time=0.9114, value=0.0806}
           {
            arriveTangent=0.6573
            arriveTangentWeight=0.3221
            interpMode=RCIM_Cubic
            leaveTangent=0.6573
            leaveTangentWeight=0.3221
            tangentMode=RCTM_User
            tangentWeightMode=RCTWM_WeightedBoth
            time=1
            value=0.1763
           }
          ]}
         RedCurve {keys [
           {
            arriveTangent=0.2622
            arriveTangentWeight=0.0786
            interpMode=RCIM_Cubic
            leaveTangent=0.2622
            leaveTangentWeight=0.0786
            tangentMode=RCTM_User
            tangentWeightMode=RCTWM_WeightedBoth
            time=0.001
            value=0.1085
           }
           <at 0.082 0.1325>
           {
            arriveTangent=1.3193
            arriveTangentWeight=0.063
            interpMode=RCIM_Cubic
            leaveTangent=1.3193
            leaveTangentWeight=0.063
            tangentMode=RCTM_User
            tangentWeightMode=RCTWM_WeightedBoth
            time=0.171
            value=0.349
           }
           <at 0.216 0.1354>
           {interpMode=RCIM_Cubic, tangentMode=RCTM_User, time=0.29, value=0.2969}
           {
            arriveTangent=1.194
            arriveTangentWeight=0.0789
            interpMode=RCIM_Cubic
            leaveTangent=1.194
            leaveTangentWeight=0.0789
            tangentMode=RCTM_User
            tangentWeightMode=RCTWM_WeightedBoth
            time=0.375
            value=0.254
           }
           {
            arriveTangent=0.0426
            arriveTangentWeight=0.1136
            interpMode=RCIM_Cubic
            leaveTangent=0.0426
            leaveTangentWeight=0.1136
            tangentMode=RCTM_User
            tangentWeightMode=RCTWM_WeightedBoth
            time=0.6589
            value=0.4193
           }
           {interpMode=RCIM_Cubic, tangentMode=RCTM_User, time=0.817, value=0.4583}
           {arriveTangent=-1.9117, interpMode=RCIM_Cubic, leaveTangent=-1.9117, time=0.9114, value=0.0219}
           {arriveTangentWeight=0.3333, interpMode=RCIM_Cubic, tangentMode=RCTM_User, tangentWeightMode=RCTWM_WeightedBoth, time=1, value=0.1085}
          ]}
        }}
       CurveIndex=Linked() {linkedVariable {name=NPC.FortniteNPC.TimeOfDay_0-1, type=NiagaraFloat}}
      ))
     B=&VectorFromFloat(Value=2)
    )
    ScaleA=true
    ScaleRGB=true
   )
   &SubUVAnimation(
    "Sprite Renderer "=DataInterface() {propertyValues {SpriteRenderer=&self:Smoke_thin_1.NiagaraSpriteRendererProperties_0}}
    Start Frame Offset:NiagaraInt32=&RandomRangeInt(Minimum=0, Maximum=4)
    UseEndFrame=false
    UseStartFrame=false
   ) {module=SubUVAnimation001}
  ]
  rendererClasses [&NiagaraSpriteRendererProperties]
  renderers [{rendererClass=&NiagaraSpriteRendererProperties}]
 }
 TrailEmbers_EPIC {
  simTarget=GPUComputeSim
  emitterUpdateScript [
   &EmitterState()
   &SpawnParticlesFromOtherEmitter(
    Attribute Reader=DataInterface() {propertyValues {EmitterBinding {bindingMode=Other, emitterName=Main}}}
    Maximum Age Spawn Threshold=2
    Minimum Age Spawn Threshold=0
    Spawn Rate=100
    Use Maximum Age=false
    Use Minimum Age=false
   )
   &SpawnParticlesFromOtherEmitter(
    Attribute Reader=DataInterface() {propertyValues {EmitterBinding {bindingMode=Other, emitterName=Main}}}
    Maximum Age Spawn Threshold=2
    Minimum Age Spawn Threshold=0
    Spawn Group=1
    Spawn Rate=15
    Spawn Rate Per Particle Cap=15
    Use Maximum Age=false
    Use Minimum Age=false
   ) {module=SpawnParticlesFromOtherEmitter001}
  ]
  particleSpawnScript [
   &InitializeParticle(
    Color=&RandomRangeLinearColor(Minimum=<rgba 25 10.323 0 1>, Maximum=<rgba 250 129.0156 0 1>)
    Lifetime Max=4
    Lifetime Mode=::NewEnumerator1
    Sprite Rotation Mode=::NewEnumerator2
    Sprite Size Mode=::NewEnumerator1
    Uniform Sprite Size Max=100
    Uniform Sprite Size Min=25
    UsePositionOffset=false
    Write Material Random=true
   )
   &AddVelocity(
    Cone Angle=2
    Cone Axis=&MakeVector(
     X=&Sine(Normalized Angle=$Emitter.Age, Period=5, Scale=1, Bias=0)
     Y=0
     Z=&RandomRangeFloat(Minimum=0, Maximum=1)
    )
    Module.UseConeFalloff=false
    Module.UseVelDistribution=false
    Override Seed=false
    Velocity Mode=::NewEnumerator2
    Velocity Speed=250
   )
   &SampleParticlesFromOtherEmitter(
    Applied Velocity Scale:NiagaraFloat=&FloatFromCurve(
     FloatCurve=DataInterface() {propertyValues {Curve {keys [<at 0 1>,<at 1 0>]}}}
     Scale Curve=&RandomRangeFloat(Minimum=0.25, Maximum=1)
    )
    Velocity Sampling=::NewEnumerator0
   )
   &SphereLocation(
    MaskSpawnGroup=false
    Override Seed=false
    Sphere Radius=&RandomRangeFloat(
     Minimum=10
     Maximum=&FloatFromCurve(FloatCurve=DataInterface() {propertyValues {Curve {keys [<at 0 1>,<at 1 0>]}}}, CurveIndex=$Emitter.NormalizedLoopAge, Scale Curve=2)
    )
    Surface Only=false
    UseNonUniformScale=false
    UseOffset=false
    UseSphereOrientation=false
   )
  ]
  particleUpdateScript [
   /SetVariables(Particles.VisibilityTag=&SpawnGroup()) {isSetParameters=true, module=SetVariables_#1}
   &Drag(
    Drag=&FloatFromCurve(FloatCurve=DataInterface() {propertyValues {Curve {keys [<at 0 0>,<at 1 1>]}}})
    Use Linear Drag=true
    Use Rotational Drag=false
   )
   &GravityForce(Gravity=<0 0 -600>)
   &CurlNoiseForce(
    Noise Frequency=1
    Noise Strength=&FloatFromCurve(
     FloatCurve=DataInterface() {propertyValues {Curve {keys [<at 0 1>,<at 1 0.5>]}}}
     Scale Curve=&RandomRangeFloat(Minimum=150, Maximum=500)
    )
    Pan Noise=false
   )
   &Collision(
    Age Colliding Particles=false
    Collision Enabled=&SetBoolByFloatComparison(A=$Particles.Lifetime, B=0.1)
    Friction=0.2
    GPU Collision Type=::NewEnumerator2
    KillOnCollision=false
    Module.Control Rest State Via Surface Normal=false
    Module.EnableMaxCollisionCount=false
    Module.Randomize Collision Normal Vector=false
    Particle Radius Scale:NiagaraFloat=&FloatFromCurve(FloatCurve=DataInterface() {propertyValues {Curve {keys [<at 0.1 0>,<at 0.25 2>]}}})
    Restitution=&RandomRangeFloat(Minimum=0, Maximum=0.25)
   )
   &SolveForcesAndVelocity(Clamp Velocity=false, Limit Acceleration=false)
   &ParticleState()
   &DynamicMaterialParameters(
    Index 0 Param 1=0.8
    Index 0 Param 2=1
    Index 0 Param 3=1
    Index 1 Param 1=150
    Index 1 Param 4=1
    Module.Param0WriteEnabled=true
    Module.Param0WriteEnabled1=true
    Module.Param1WriteEnabled=true
    Module.Param1WriteEnabled1=true
    Module.Param2WriteEnabled=true
    Module.Param2WriteEnabled1=true
    Module.Param3WriteEnabled=false
    Module.Param3WriteEnabled1=true
    Write Parameter Index 1=true
   )
   &ScaleColor(
    Scale RGB=&Multiply_VectorByFloat(
     Vector=<12 2 0.05>
     Float=&FloatFromCurve(
      FloatCurve=DataInterface() {propertyValues {Curve {keys [{interpMode=RCIM_Cubic, tangentMode=RCTM_User, time=0, value=1}, {interpMode=RCIM_Cubic, tangentMode=RCTM_User, time=1, value=0.5}]}}}
      Scale Curve=&Waveform(
       "Waveform [1]"=::NewEnumerator0
       "[1] Phase Scale"=&RandomRangeFloat(Minimum=1, Maximum=2)
       "[1] Frequency"=&Lerp_Float(
        A=&RandomRangeFloat(Minimum=4, Maximum=5)
        B=1
        Alpha=&MaskFloatBySpawnGroup(Masked Float=1, Passthrough Float=1, Spawn Group Mask=1)
       )
       "[1] Amplitude Min/Max"=<1 2>
      )
     )
    )
    ScaleA=true
    ScaleRGB=true
   ) {module=ScaleColor001}
   /Niagara/Modules/Update/Camera/MaintainInCameraParticleScale(Pixel Depth=0.1, Render Target Width=0.1)
   &ScaleSpriteSizeBySpeed(
    Max Scale Factor=<0.2 0.6>
    Min Scale Factor=<0.1 0.1>
    Sample Scale Factor By Curve=true
    Scale Factor Curve=DataInterface() {propertyValues {Curve {keys [
        {interpMode=RCIM_Cubic, tangentMode=RCTM_User, time=0, value=0}
        {arriveTangent=2.9664, interpMode=RCIM_Cubic, leaveTangent=2.9664, tangentMode=RCTM_User, time=1, value=1}
       ]}}}
    Velocity Threshold=100
   )
   &ScaleSpriteSize(
    Uniform Curve Scale=&Multiply_Float(
     A=&Multiply_Float(
      A=0.2
      B=&Waveform(
       "Waveform [1]"=::NewEnumerator0
       "[1] Phase Scale"=&RandomRangeFloat(Minimum=3, Maximum=10)
       "[1] Frequency"=&RandomRangeFloat(Minimum=0.2, Maximum=0.5)
       "[1] Amplitude Min/Max"=<0.25 1>
      )
     )
     B=&FN_LOD_Distance_Remap(Start Distance=500, End Distance=5000, Minimum Output=0.2, Maximum Output=1, Clamp Output=true)
    )
    Uniform Curve Sprite Scale=DataInterface() {propertyValues {Curve {keys [
        {arriveTangent=2.9868, interpMode=RCIM_Cubic, leaveTangent=2.9868, tangentMode=RCTM_User, time=0, value=1.65}
        <at 0.4 2.3>
        {arriveTangent=-0.0405, interpMode=RCIM_Cubic, leaveTangent=-0.0405, tangentMode=RCTM_User, time=1, value=1.4}
       ]}}}
   )
   /SetVariables(
    Particles.Linear Color=&MultiplyLinearColor(Linear Color=$Particles.Color, Scale Factor=<0.02 0.02 0.02 0.0001>)
    Particles.Vector:Vector3f=&Multiply_Vector(A=$Particles.Velocity, B=<1 1 1>)
    Particles.Vector 2D:Vector2f=&Multiply_Vector2D(A=<2 3>, B=$Particles.SpriteSize)
   ) {isSetParameters=true, module=SetVariables_#2}
  ]
  rendererClasses [&NiagaraSpriteRendererProperties]
  renderers [
   {rendererClass=&NiagaraSpriteRendererProperties}
   {rendererClass=&NiagaraSpriteRendererProperties, rendererIndex=1}
  ]
 }
 TrailEmbers_HIGH {
  simTarget=GPUComputeSim
  emitterUpdateScript [
   &EmitterState()
   &SpawnParticlesFromOtherEmitter(
    Attribute Reader=DataInterface() {propertyValues {EmitterBinding {bindingMode=Other, emitterName=Main}}}
    Maximum Age Spawn Threshold=2
    Minimum Age Spawn Threshold=0
    Spawn Rate=50
    Spawn Rate Per Particle Cap=50
    Use Maximum Age=false
    Use Minimum Age=false
   )
   &SpawnParticlesFromOtherEmitter(
    Attribute Reader=DataInterface() {propertyValues {EmitterBinding {bindingMode=Other, emitterName=Main}}}
    Maximum Age Spawn Threshold=2
    Minimum Age Spawn Threshold=0
    Spawn Group=1
    Spawn Rate Per Particle Cap=10
    Use Maximum Age=false
    Use Minimum Age=false
   ) {module=SpawnParticlesFromOtherEmitter001}
  ]
  particleSpawnScript [
   &InitializeParticle(
    Color=&RandomRangeLinearColor(Minimum=<rgba 25 10.323 0 1>, Maximum=<rgba 250 129.0156 0 1>)
    Lifetime Max=4
    Lifetime Mode=::NewEnumerator1
    Sprite Rotation Mode=::NewEnumerator2
    Sprite Size Mode=::NewEnumerator1
    Uniform Sprite Size Max=100
    Uniform Sprite Size Min=25
    UsePositionOffset=false
    Write Material Random=true
   )
   &AddVelocity(
    Cone Angle=2
    Cone Axis=&MakeVector(
     X=&Sine(Normalized Angle=$Emitter.Age, Period=5, Scale=1, Bias=0)
     Y=0
     Z=&RandomRangeFloat(Minimum=0, Maximum=1)
    )
    Module.UseConeFalloff=false
    Module.UseVelDistribution=false
    Override Seed=false
    Velocity Mode=::NewEnumerator2
    Velocity Speed=250
   )
   &SampleParticlesFromOtherEmitter(
    Applied Velocity Scale:NiagaraFloat=&FloatFromCurve(
     FloatCurve=DataInterface() {propertyValues {Curve {keys [<at 0 1>,<at 1 0>]}}}
     Scale Curve=&RandomRangeFloat(Minimum=0.25, Maximum=1)
    )
    Velocity Sampling=::NewEnumerator0
   )
   &SphereLocation(
    MaskSpawnGroup=false
    Override Seed=false
    Sphere Radius=&RandomRangeFloat(
     Minimum=10
     Maximum=&FloatFromCurve(FloatCurve=DataInterface() {propertyValues {Curve {keys [<at 0 1>,<at 1 0>]}}}, CurveIndex=$Emitter.NormalizedLoopAge, Scale Curve=2)
    )
    Surface Only=false
    UseNonUniformScale=false
    UseOffset=false
    UseSphereOrientation=false
   )
  ]
  particleUpdateScript [
   /SetVariables(Particles.VisibilityTag=&SpawnGroup()) {isSetParameters=true, module=SetVariables_#1}
   &Drag(
    Drag=&FloatFromCurve(FloatCurve=DataInterface() {propertyValues {Curve {keys [<at 0 0>,<at 1 1>]}}})
    Use Linear Drag=true
    Use Rotational Drag=false
   )
   &GravityForce(Gravity=<0 0 -600>)
   &CurlNoiseForce(
    Noise Frequency=1
    Noise Strength=&FloatFromCurve(
     FloatCurve=DataInterface() {propertyValues {Curve {keys [<at 0 1>,<at 1 0.5>]}}}
     Scale Curve=&RandomRangeFloat(Minimum=150, Maximum=500)
    )
    Pan Noise=false
   )
   &Collision(
    Age Colliding Particles=false
    Collision Enabled=&SetBoolByFloatComparison(A=$Particles.Lifetime, B=0.1)
    Friction=0.2
    GPU Collision Type=::NewEnumerator2
    KillOnCollision=false
    Module.Control Rest State Via Surface Normal=false
    Module.EnableMaxCollisionCount=false
    Module.Randomize Collision Normal Vector=false
    Particle Radius Scale:NiagaraFloat=&FloatFromCurve(FloatCurve=DataInterface() {propertyValues {Curve {keys [<at 0.1 0>,<at 0.25 2>]}}})
    Restitution=&RandomRangeFloat(Minimum=0, Maximum=0.25)
   )
   &SolveForcesAndVelocity(Clamp Velocity=false, Limit Acceleration=false)
   &ParticleState()
   &DynamicMaterialParameters(
    Index 0 Param 1=0.8
    Index 0 Param 2=1
    Index 0 Param 3=1
    Index 1 Param 1=150
    Index 1 Param 4=1
    Module.Param0WriteEnabled=true
    Module.Param0WriteEnabled1=true
    Module.Param1WriteEnabled=true
    Module.Param1WriteEnabled1=true
    Module.Param2WriteEnabled=true
    Module.Param2WriteEnabled1=true
    Module.Param3WriteEnabled=false
    Module.Param3WriteEnabled1=true
    Write Parameter Index 1=true
   )
   &ScaleColor(
    Scale RGB=&Multiply_VectorByFloat(
     Vector=<12 2 0.05>
     Float=&FloatFromCurve(
      FloatCurve=DataInterface() {propertyValues {Curve {keys [{interpMode=RCIM_Cubic, tangentMode=RCTM_User, time=0, value=1}, {interpMode=RCIM_Cubic, tangentMode=RCTM_User, time=1, value=0.5}]}}}
      Scale Curve=&Waveform(
       "Waveform [1]"=::NewEnumerator0
       "[1] Phase Scale"=&RandomRangeFloat(Minimum=1, Maximum=2)
       "[1] Frequency"=&Lerp_Float(
        A=&RandomRangeFloat(Minimum=4, Maximum=5)
        B=1
        Alpha=&MaskFloatBySpawnGroup(Masked Float=1, Passthrough Float=1, Spawn Group Mask=1)
       )
       "[1] Amplitude Min/Max"=<1 2>
      )
     )
    )
    ScaleA=true
    ScaleRGB=true
   ) {module=ScaleColor001}
   &MaintainInCameraParticleScale(Pixel Depth=0.1, Render Target Width=0.1)
   &ScaleSpriteSizeBySpeed(
    Max Scale Factor=<0.2 0.6>
    Min Scale Factor=<0.1 0.1>
    Sample Scale Factor By Curve=true
    Scale Factor Curve=DataInterface() {propertyValues {Curve {keys [
        {interpMode=RCIM_Cubic, tangentMode=RCTM_User, time=0, value=0}
        {arriveTangent=2.9664, interpMode=RCIM_Cubic, leaveTangent=2.9664, tangentMode=RCTM_User, time=1, value=1}
       ]}}}
    Velocity Threshold=100
   )
   &ScaleSpriteSize(
    Uniform Curve Scale=&Multiply_Float(
     A=&Multiply_Float(
      A=0.2
      B=&Waveform(
       "Waveform [1]"=::NewEnumerator0
       "[1] Phase Scale"=&RandomRangeFloat(Minimum=3, Maximum=10)
       "[1] Frequency"=&RandomRangeFloat(Minimum=0.2, Maximum=0.5)
       "[1] Amplitude Min/Max"=<0.25 1>
      )
     )
     B=&FN_LOD_Distance_Remap(Start Distance=500, End Distance=5000, Minimum Output=0.2, Maximum Output=1, Clamp Output=true)
    )
    Uniform Curve Sprite Scale=DataInterface() {propertyValues {Curve {keys [
        {arriveTangent=2.9868, interpMode=RCIM_Cubic, leaveTangent=2.9868, tangentMode=RCTM_User, time=0, value=1.65}
        <at 0.4 2.3>
        {arriveTangent=-0.0405, interpMode=RCIM_Cubic, leaveTangent=-0.0405, tangentMode=RCTM_User, time=1, value=1.4}
       ]}}}
   )
   /SetVariables(
    Particles.Linear Color=&MultiplyLinearColor(Linear Color=$Particles.Color, Scale Factor=<0.02 0.02 0.02 0.0001>)
    Particles.Vector:Vector3f=&Multiply_Vector(A=$Particles.Velocity, B=<1 1 1>)
    Particles.Vector 2D:Vector2f=&Multiply_Vector2D(A=<2 3>, B=$Particles.SpriteSize)
   ) {isSetParameters=true, module=SetVariables_#2}
  ]
  rendererClasses [&NiagaraSpriteRendererProperties]
  renderers [
   {rendererClass=&NiagaraSpriteRendererProperties}
   {rendererClass=&NiagaraSpriteRendererProperties, rendererIndex=1}
  ]
 }
}
summary {
 systemName=NS_DragonCart_BusFX
 emitters [
  {emitterName=Smoke_thin, rendererClasses [&NiagaraSpriteRendererProperties]}
  {emitterName=Main, simTarget=GPUComputeSim, rendererClasses [&NiagaraSpriteRendererProperties]}
  {emitterName=TrailEmbers_EPIC, simTarget=GPUComputeSim, rendererClasses [&NiagaraSpriteRendererProperties]}
  {emitterName=TrailEmbers_HIGH, simTarget=GPUComputeSim, rendererClasses [&NiagaraSpriteRendererProperties]}
 ]
 userVariables [User.Speed {type=NiagaraFloat, defaultValue=0, description=""}]
}
