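# Niagara effect that streams magical shards and sparkles from a world start position to attach onto a target static mesh.
# GPU shard meshes lerp toward sampled surface points with normal-based arcs and sine bobbing, flaring brighter before vanishing.
# Sparkles drift similarly toward the owner, and normalized travel distance lengthens lifetimes and thins spawning for both.
# Swirl meshes add distance-faded accents driven by dynamic material curves, while PlasmaWave has no renderers.
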NiagaraSystem /DragonCartNarrative/Assets/VFX/NS_DragonCart_Shard_Attach
dependencies {
 usedDataInterfaces [
  &NiagaraDataInterfaceArrayInt32
  &NiagaraDataInterfaceMeshRendererInfo
  &NiagaraDataInterfaceCurve
  &NiagaraDataInterfaceArrayColor
 ]
 usedDynamicInputs [
  &RandomVector
  &Vector2DFromFloat
  &DistanceBetweenPositions
 ]
 usedModules [
  &self:SystemSpawnScript.SystemScriptSource.SystemScriptGraph.NiagaraNodeAssignment_0.SetVariables_#1
  &SystemState
  &self:SystemSpawnScript.SystemScriptSource.SystemScriptGraph.NiagaraNodeAssignment_1.SetVariables_#2
 ]
}
emitters {
 Mesh_Swirls {
  emitterUpdateScript [
   &EmitterState()
   &SpawnRate(SpawnRate=10, Use Spawn Probability=false)
  ]
  particleSpawnScript [
   &InitializeParticle(
    Color=&SelectLinearColorFromArray(
     Array Sampling Mode=::NewEnumerator1
     Direct Array Index=&RandomRangeInt(Minimum=0, Maximum=1)
     Color Selection Array=DataInterface() {propertyValues {ColorData [<rgba 0.3199 0.5916 1 1>, <rgba 0.9443 0.3385 1 1>]}}
     Random Seed=?
    )
    Lifetime=&RandomRangeFloat(Minimum=3, Maximum=4)
    Mesh Scale=&RandomRangeVector(Minimum=<1 1.5 1.5>, Maximum=<2 8 8>)
    Mesh Scale Mode=::NewEnumerator4
    UsePositionOffset=false
    Write Material Random=true
   ) {module=InitializeParticle001}
   &InitialMeshOrientation(
    Rotation=&RandomRangeVector(Minimum=<-1 -1 -1>, Maximum=<1 1 1>)
    Rotation Coordinate Space=::NewEnumerator2
    Use Orientation Vector=true
    Use Rotation Vector=true
   ) {module=InitialMeshOrientation001}
   &ApplyInitialForces()
   &AddVelocity(Module.Use Falloff Distance=false, Override Seed=false, Velocity Mode=::NewEnumerator1, Velocity Speed=5)
  ]
  particleUpdateScript [
   &ParticleState()
   &DynamicMaterialParameters(
    Index 0 Param 1=&RandomRangeFloat(Minimum=0, Maximum=1)
    Index 0 Param 2=&FloatFromCurve(FloatCurve=DataInterface() {propertyValues {Curve {
        defaultValue=1
        keys [
         {interpMode=RCIM_Cubic, tangentMode=RCTM_User, time=0, value=1}
         <at 0.5 0 RCIM_Cubic>
         {interpMode=RCIM_Cubic, tangentMode=RCTM_User, time=1, value=1}
        ]
       }}})
    Index 0 Param 3=0.5
    Index 0 Param 4=&FloatFromCurve(
     FloatCurve=DataInterface() {propertyValues {Curve {keys [<at 0 0.3 RCIM_Cubic>, {arriveTangent=-3.2559, interpMode=RCIM_Cubic, leaveTangent=-3.2559, tangentMode=RCTM_User, time=1, value=-1}]}}}
     Scale Curve=0.5
    )
    Module.Param0WriteEnabled=true
    Module.Param1WriteEnabled=true
    Module.Param2WriteEnabled=true
    Module.Param3WriteEnabled=true
   )
   &ScaleColor(
    Scale Alpha=&FloatFromCurve(
     FloatCurve=DataInterface() {propertyValues {Curve {keys [
         {arriveTangent=12.4054, interpMode=RCIM_Cubic, leaveTangent=12.4054, tangentMode=RCTM_User, time=0.125, value=0}
         <at 0.2 1 RCIM_Cubic>
         {arriveTangent=-0.7128, interpMode=RCIM_Cubic, leaveTangent=-0.7128, tangentMode=RCTM_User, time=1, value=0}
        ]}}}
     Scale Curve=&Multiply_Float(
      A=&RandomRangeFloat(Minimum=0.5, Maximum=1)
      B=&FN_LOD_Distance_Remap(Start Distance=500, End Distance=2500, Minimum Output=0, Maximum Output=1, Clamp Output=true)
     )
    )
    ScaleA=true
    ScaleRGB=true
   )
   &ScaleMeshSize(Scale Factor=&VectorFromFloat(Value=&FloatFromCurve(
      FloatCurve=DataInterface() {propertyValues {Curve {keys [
          {arriveTangent=2.61, interpMode=RCIM_Cubic, leaveTangent=2.61, tangentMode=RCTM_User, time=0, value=0}
          {interpMode=RCIM_Cubic, leaveTangent=-0.8889, time=0.8, value=1}
         ]}}}
      Scale Curve=0.5
     )))
   &SolveForcesAndVelocity(Clamp Velocity=false, Limit Acceleration=false)
  ]
  rendererClasses [&NiagaraMeshRendererProperties]
  renderers [{rendererClass=&NiagaraMeshRendererProperties, rendererIndex=0}]
 }
 PlasmaWave {rendererClasses [], renderers []}
 ShardMeshes {
  simTarget=GPUComputeSim
  emitterUpdateScript [
   &EmitterState(Enable Distance Culling=true, MaxDistance=8000, Module.bUseMaxDistance=true, Module.bUseMinDistance=false, Scalability Mode=::NewEnumerator1)
   &SpawnRate(
    Spawn Probability=0.4
    SpawnRate=&Lerp_Float(A=1, B=0.25, Alpha=$System.TravelDistanceNormalized)
    Use Spawn Probability=true
   )
  ]
  particleSpawnScript [
   &InitializeParticle(
    Color=<rgba 0.7708 0.156 1 1>
    Lifetime=&Lerp_Float(
     A=&RandomRangeFloat(Minimum=9, Maximum=12)
     B=&RandomRangeFloat(Minimum=17, Maximum=20)
     Alpha=$System.TravelDistanceNormalized
    )
    Mesh Renderer Array Visibility Mode=::NewEnumerator1
    Mesh Renderer Info=DataInterface() {propertyValues {MeshRenderer=&self:ShardMeshes_0.NiagaraMeshRendererProperties_0}}
    Mesh Renderer Mesh Index:NiagaraInt32=&SelectIntFromArray(
     Array Sampling Mode=::NewEnumerator1
     Direct Array Index=?
     Int Selection Array=DataInterface() {propertyValues {IntData [0,1,2,3,0,1,2]}}
     Random Seed=0
    )
    Mesh Scale Max=<2.2 2.2 2.2>
    Mesh Scale Min=<1.2 1.2 1.2>
    Mesh Scale Mode=::NewEnumerator2
    UsePositionOffset=false
    Write Material Random=true
   )
   /Niagara/Modules/Spawn/Location/StaticMeshLocation(
    MaskSpawnGroup=false
    OffsetAlongNormal=false
    Orientation Sampling:ENiagara_AttributeSamplingApplyOutput=::NewEnumerator1
    OverrideRandomSeed=false
    Position Sampling:ENiagaraSkelMeshPositionSamplingMode=::NewEnumerator2
    SamplePosOffset=false
    Static Mesh:NiagaraDataInterfaceStaticMesh=$User.Static Mesh Reader
   )
   /SetVariables(Particles.Position=&AddVectorToPosition(
     Position=$User.WorldStartPosition
     Vector=&Multiply_VectorByFloat(
      Vector=&NormalizeVector(
       Vector To Normalize=&Multiply_Vector(A=$Particles.StaticMeshLocation.SampledNormal, B=<1 1 0>)
       Safe Normalize=true
       Safe Fallback Threshold=0
       Safe Fallback Vector=<0 0 0.001>
      )
      Float=20
     )
    )) {isSetParameters=true}
   &InitialMeshOrientation(
    Mesh Orientation Mode=::NewEnumerator6
    Orientation Quaternion:Quat4f=$Particles.StaticMeshLocation.SampledRotation
    Rotation=&RandomRangeVector(Minimum=<0 0 0>, Maximum=<1 0 0>)
    Use Rotation Vector=true
   )
  ]
  particleUpdateScript [
   &ParticleState()
   &ScaleMeshSize(Scale Factor=&VectorFromFloat(Value=&FloatFromCurve(
      FloatCurve=DataInterface() {propertyValues {Curve {keys [
          {arriveTangent=36.4918, interpMode=RCIM_Cubic, leaveTangent=36.4918, tangentMode=RCTM_User, time=0, value=0}
          {arriveTangentWeight=0.0242, interpMode=RCIM_Cubic, leaveTangentWeight=0.4356, tangentMode=RCTM_User, time=0.05, value=1}
          {arriveTangentWeight=0.4356, interpMode=RCIM_Cubic, leaveTangentWeight=0.0242, tangentMode=RCTM_User, time=0.95, value=1}
          {arriveTangent=-27.6295, tangentMode=RCTM_User, time=1, value=0}
         ]}}}
      Scale Curve=$System.ScaleLerp
     )))
   &UpdateMeshOrientation(
    Rotation Coordinate Space=::NewEnumerator0
    Rotation Rate=&FloatFromCurve(
     FloatCurve=DataInterface() {propertyValues {Curve {keys [
         {arriveTangent=1.074, interpMode=RCIM_Cubic, time=0.3, value=0.4}
         {
          arriveTangent=1.3652
          arriveTangentWeight=0.2637
          interpMode=RCIM_Cubic
          leaveTangent=1.3652
          leaveTangentWeight=0.0942
          tangentMode=RCTM_User
          time=0.6
          value=1
         }
         <at 0.95 0>
        ]}}}
     Scale Curve=&RandomRangeFloat(Minimum=0.3, Maximum=0.5)
    )
    Rotation Vector=$Particles.StaticMeshLocation.SampledNormal
   )
   /SetVariables(Particles.Position=&AddVectorToPosition(
     Position=&Lerp_Vector(
      A=Linked() {linkedVariable {name=Particles.Initial.Position, type=NiagaraPosition}}
      B=Linked() {linkedVariable {name=Particles.StaticMeshLocation.SampledPosition, type=NiagaraPosition}}
      Alpha=&FloatFromCurve(FloatCurve=DataInterface() {propertyValues {Curve {
          defaultValue=1
          keys [
           {arriveTangent=1.3367, interpMode=RCIM_Cubic, leaveTangent=1.3367, tangentMode=RCTM_User, time=0, value=0}
           {interpMode=RCIM_Cubic, leaveTangent=0.8453, time=0.95, value=1}
          ]
         }}})
     )
     Vector=&Add_Vector(
      A=&Multiply_VectorByFloat(
       Vector=&NormalizeVector(
        Vector To Normalize=&Multiply_Vector(A=$Particles.StaticMeshLocation.SampledNormal, B=<1 1 0.3>)
        Safe Normalize=true
        Safe Fallback Threshold=0
        Safe Fallback Vector=<0 0 0.001>
       )
       Float=&FloatFromCurve(
        FloatCurve=DataInterface() {propertyValues {Curve {
           defaultValue=0
           keys [
            <at 0 0 RCIM_Cubic>
            {arriveTangent=1.4286, interpMode=RCIM_Cubic, leaveTangent=1.4286, time=0.3, value=-0.2}
            {arriveTangent=-0.4597, interpMode=RCIM_Cubic, leaveTangent=-0.4597, tangentMode=RCTM_User, time=0.7, value=1}
            <at 0.95 -0.05 RCIM_Cubic>
           ]
          }}}
        Scale Curve=120
       )
      )
      B=&MakeVector(
       X=0
       Y=0
       Z=&Waveform(
        Global Phase=&Multiply_Float(A=$Engine.Time, B=0.2)
        Global Amplitude Scale=-8
        "Waveform [1]"=::NewEnumerator0
        "[1] Frequency"=1
        "[1] Amplitude Min/Max"=<-1 1>
       )
      )
     )
    )) {isSetParameters=true}
   &ScaleColor(
    Scale RGB=&Lerp_Vector(
     A=<1 1 1>
     B=<20 5 1>
     Alpha=&FloatFromCurve(FloatCurve=DataInterface() {propertyValues {Curve {defaultValue=1, keys [<at 0.87 0 RCIM_Cubic>,<at 0.96 1 RCIM_Cubic>]}}}, Scale Curve=5)
    )
    ScaleA=true
    ScaleRGB=true
   )
  ]
  rendererClasses [&NiagaraMeshRendererProperties]
  renderers [{rendererClass=&NiagaraMeshRendererProperties, rendererIndex=0}]
 }
 Sparkles {
  emitterUpdateScript [
   &EmitterState(Enable Distance Culling=true, MaxDistance=6000, Module.bUseMaxDistance=true, Module.bUseMinDistance=false, Scalability Mode=::NewEnumerator1)
   &SpawnRate(
    Spawn Probability=0.4
    SpawnRate=&Lerp_Float(A=8, B=4, Alpha=$System.TravelDistanceNormalized)
    Use Spawn Probability=true
   )
  ]
  particleSpawnScript [
   &InitializeParticle(
    Color Maximum=<rgba 8 0.4638 5.1406 1>
    Color Minimum=<rgba 0.651 0.8702 5 1>
    Color Mode=::NewEnumerator2
    Lifetime=&Lerp_Float(
     A=&RandomRangeFloat(Minimum=7, Maximum=14)
     B=&RandomRangeFloat(Minimum=15, Maximum=22)
     Alpha=$System.TravelDistanceNormalized
    )
    Position=$User.WorldStartPosition
    Position Mode=::NewEnumerator1
    Sprite Rotation Mode=::NewEnumerator2
    Sprite Size Mode=::NewEnumerator1
    Uniform Sprite Size Max=30
    Uniform Sprite Size Min=16
    UsePositionOffset=false
    Write Material Random=true
   )
   &ShapeLocation(MaskSpawnGroup=false, Override Seed=false, Sphere Radius=300, UseOwnerScale=true)
  ]
  particleUpdateScript [
   &ParticleState()
   &ScaleSpriteSize(
    Uniform Curve Scale=$System.ScaleLerp
    Uniform Curve Sprite Scale=DataInterface() {propertyValues {Curve {keys [
        {arriveTangent=36.4918, interpMode=RCIM_Cubic, leaveTangent=36.4918, tangentMode=RCTM_User, time=0, value=0}
        {arriveTangentWeight=0.0242, interpMode=RCIM_Cubic, leaveTangentWeight=0.4356, tangentMode=RCTM_User, time=0.05, value=1}
        {arriveTangentWeight=0.4356, interpMode=RCIM_Cubic, leaveTangentWeight=0.0242, tangentMode=RCTM_User, time=0.95, value=1}
        {arriveTangent=-27.6295, tangentMode=RCTM_User, time=1, value=0}
       ]}}}
   )
   /SetVariables(Particles.Position=&AddVectorToPosition(
     Position=&Lerp_Vector(
      A=Linked() {linkedVariable {name=Particles.Initial.Position, type=NiagaraPosition}}
      B=Linked() {linkedVariable {name=Engine.Owner.Position, type=NiagaraPosition}}
      Alpha=&FloatFromCurve(FloatCurve=DataInterface() {propertyValues {Curve {
          defaultValue=1
          keys [
           {arriveTangent=1.3367, interpMode=RCIM_Cubic, leaveTangent=1.3367, tangentMode=RCTM_User, time=0, value=0}
           {arriveTangent=0.4398, interpMode=RCIM_Cubic, leaveTangent=0.4398, tangentMode=RCTM_User, time=1, value=1}
          ]
         }}})
     )
     Vector=&Add_Vector(
      A=&Multiply_VectorByFloat(
       Vector=$Particles.ShapeLocation.ShapeNormal
       Float=&FloatFromCurve(
        FloatCurve=DataInterface() {propertyValues {Curve {
           defaultValue=0
           keys [
            <at 0 0 RCIM_Cubic>
            {arriveTangent=1.4286, interpMode=RCIM_Cubic, leaveTangent=1.4286, time=0.3, value=-0.1}
            {arriveTangent=-0.4597, interpMode=RCIM_Cubic, leaveTangent=-0.4597, tangentMode=RCTM_User, time=0.7, value=1}
            {arriveTangent=-4.7939, interpMode=RCIM_Cubic, leaveTangent=-4.7939, tangentMode=RCTM_User, time=1, value=0.4}
           ]
          }}}
        Scale Curve=250
       )
      )
      B=&MakeVector(
       X=0
       Y=0
       Z=&Waveform(
        Global Phase=&Multiply_Float(A=$Engine.Time, B=0.4)
        Global Amplitude Scale=-8
        "Waveform [1]"=::NewEnumerator0
        "[1] Frequency"=1
        "[1] Amplitude Min/Max"=<-1 1>
       )
      )
     )
    )) {isSetParameters=true}
  ]
  rendererClasses [&NiagaraSpriteRendererProperties]
  renderers [{rendererClass=&NiagaraSpriteRendererProperties}]
 }
}
summary {
 systemName=NS_DragonCart_Shard_Attach
 emitters [
  {emitterName=ShardMeshes, simTarget=GPUComputeSim, rendererClasses [&NiagaraMeshRendererProperties]}
  {emitterName=Sparkles, rendererClasses [&NiagaraSpriteRendererProperties]}
  {emitterName=Mesh_Swirls, rendererClasses [&NiagaraMeshRendererProperties]}
  {emitterName=PlasmaWave, rendererClasses []}
 ]
 userVariables [
  User.Static Mesh Reader {
   type=NiagaraDataInterfaceStaticMesh
   description=""
   defaultValue {
    struct=NiagaraExt_VariableValue_DataInterface
    value {
     dataInterface=&self:NiagaraDataInterfaceStaticMesh_0
     dataInterfaceClass=&NiagaraDataInterfaceStaticMesh
    }
   }
  }
  User.WorldStartPosition {type=NiagaraPosition, description="", defaultValue {}}
 ]
}
